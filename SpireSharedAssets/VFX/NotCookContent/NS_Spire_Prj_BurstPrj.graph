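# Niagara system NS_Spire_Prj_BurstPrj, by its name a burst projectile effect for Spire, stored under NotCookContent.
# Six emitters make up the effect: three flare layers, a dark layer, lightning and sparks, going by their names.
# The system script uses SystemState plus two variable assignment modules.
# It lists no user variables, renderers, data interfaces or dynamic inputs, so emitter visuals and tunable inputs are not shown.
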NiagaraSystem /SpireSharedAssets/VFX/NotCookContent/NS_Spire_Prj_BurstPrj
dependencies {
 usedDataInterfaces []
 usedDynamicInputs []
 usedModules [
  &self:SystemSpawnScript.SystemScriptSource.SystemScriptGraph.NiagaraNodeAssignment_0.SetVariables_#1
  &self:SystemSpawnScript.SystemScriptSource.SystemScriptGraph.NiagaraNodeAssignment_1.SetVariables_#2
  &SystemState
 ]
}
emitters {
 Dark {rendererClasses [], renderers []}
 Flare {rendererClasses [], renderers []}
 Flare001 {rendererClasses [], renderers []}
 Flare003 {rendererClasses [], renderers []}
 LW_Lightning001 {rendererClasses [], renderers []}
 Sparks {rendererClasses [], renderers []}
}
summary {
 systemName=NS_Spire_Prj_BurstPrj
 emitters [
  {emitterName=Flare003, rendererClasses []}
  {emitterName=Dark, rendererClasses []}
  {emitterName=LW_Lightning001, rendererClasses []}
  {emitterName=Flare, rendererClasses []}
  {emitterName=Sparks, rendererClasses []}
  {emitterName=Flare001, rendererClasses []}
 ]
 userVariables []
}
