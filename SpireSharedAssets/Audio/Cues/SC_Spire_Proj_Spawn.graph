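# Sound cue played when a Spire projectile spawns, layering a spawn sound with a build-up in the boss-fight projectile sound class.
# A mixer combines one random spawn variant with one random major-spire build-up variant, the latter starting after a short fixed delay; each layer passes through a modulator.
# It borrows a hit-marker attenuation asset and overrides concurrency with a small per-owner limit that stops the oldest instance.
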SoundCue /SpireSharedAssets/Audio/Cues/SC_Spire_Proj_Spawn
root=Mixer_0
nodes [
 {id=Delay_0, children [Modulator_1], properties {delayMax=0.1, delayMin=0.1}}
 {id=Mixer_0, children [Modulator_0,Delay_0], properties {inputVolume [1,1]}}
 {id=Modulator_0, children [Random_0]}
 {id=Modulator_1, children [Random_1]}
 {
  id=Random_0
  children [WavePlayer_1,WavePlayer_2,WavePlayer_3,WavePlayer_4,WavePlayer_5]
  properties {bShouldExcludeFromBranchCulling=true, weights [1,1,1,1,1]}
 }
 {id=Random_1, children [WavePlayer_0,WavePlayer_6,WavePlayer_7,WavePlayer_8], properties {weights [1,1,1,1]}}
 {id=WavePlayer_0, properties {soundWaveAssetPtr=/SpireSharedAssets/Audio/Waves/MajorSpire_Proj_BuildUp_01}}
 {id=WavePlayer_1, properties {soundWaveAssetPtr=/SpireSharedAssets/Audio/Waves/Spire_Projectile_Spawn_01}}
 {id=WavePlayer_2, properties {soundWaveAssetPtr=/SpireSharedAssets/Audio/Waves/Spire_Projectile_Spawn_02}}
 {id=WavePlayer_3, properties {soundWaveAssetPtr=/SpireSharedAssets/Audio/Waves/Spire_Projectile_Spawn_03}}
 {id=WavePlayer_4, properties {soundWaveAssetPtr=/SpireSharedAssets/Audio/Waves/Spire_Projectile_Spawn_04}}
 {id=WavePlayer_5, properties {soundWaveAssetPtr=/SpireSharedAssets/Audio/Waves/Spire_Projectile_Spawn_05}}
 {id=WavePlayer_6, properties {soundWaveAssetPtr=/SpireSharedAssets/Audio/Waves/MajorSpire_Proj_BuildUp_02}}
 {id=WavePlayer_7, properties {soundWaveAssetPtr=/SpireSharedAssets/Audio/Waves/MajorSpire_Proj_BuildUp_03}}
 {id=WavePlayer_8, properties {soundWaveAssetPtr=/SpireSharedAssets/Audio/Waves/MajorSpire_Proj_BuildUp_04}}
]
properties {
 attenuationSettings=/Game/Athena/Sounds/HitNotifications/Fort_HitMarker_Instigate_Att
 bOverrideConcurrency=true
 maxDistance=2097152
 soundClassObject=/Game/Sounds/FortSoundClassesAndModes/Fort_BossFight_SoundClass/Fort_BossFight_Projectile
 volumeMultiplier=0.5012
 concurrencyOverrides {bLimitToOwner=true, maxCount=3, resolutionRule=StopOldest, retriggerTime=1, volumeScaleAttackTime=0.3}
}
